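; An emote anim notify state (MicroLeaf folder) that bumps a scalar material parameter by one on a chosen attached skeletal mesh when the notify begins.
; It searches all descendant components of the animated mesh for skeletal mesh components using the configured mesh asset.
; On each match, it reads the named parameter from a dynamic material instance at the configured index, rounds it, increments it and writes it back.
; Debug prints report the new value or an unset mesh.
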(fn Received_NotifyBegin
  (params (MeshComp :type Skeletal%20Mesh%20Component%20Object%20Reference) (Animation :type Anim%20Sequence%20Base%20Object%20Reference) (TotalDuration :type Float%20%28single-precision%29) (EventReference :type Anim%20Notify%20Event%20Reference%20Structure%20%28by%20ref%29))
  (bind _materialindex (Variables|Default|GetMaterialIndex))
  (Variables|Default|GetSkeletalMesh
    (:then
      (for _array_element (Components|GetChildrenComponents MeshComp true)
        (bind _asskeletal_mesh_component (Utilities|Casting|CastToSkeletalMeshComponent _array_element))
        (if (== (Variables|Default|GetSkeletalMesh) (Variables|/Script/Engine.SkinnedMeshComponent|GetSkeletalMesh _asskeletal_mesh_component))
          (bind _returnvalue (Rendering|Material|GetScalarParameterValue (Class|BTechArtGlobalLibrary|GetDynamicMaterialInstance _asskeletal_mesh_component _materialindex _asskeletal_mesh_component) (Variables|Default|GetMaterialParameter) :!class /Script/Engine.MaterialInstanceDynamic))
          (bind _result (Math|Integer|IncrementInt (Math|Float|Round _returnvalue))
            (:%20%20
              (Rendering|Material|SetScalarParameterValue (Class|BTechArtGlobalLibrary|GetDynamicMaterialInstance _asskeletal_mesh_component _materialindex _asskeletal_mesh_component) (Variables|Default|GetMaterialParameter) (Math|Conversions|ToFloat(Integer) _result) :!class /Script/Engine.MaterialInstanceDynamic)
              (Development|PrintString 0 (Utilities|String|ToString(Float) (Math|Conversions|ToFloat(Integer) _result))))))))
    (:else
      (Development|PrintString 0 "Please select a static mesh on the AnimNotify")))
  (return false))
