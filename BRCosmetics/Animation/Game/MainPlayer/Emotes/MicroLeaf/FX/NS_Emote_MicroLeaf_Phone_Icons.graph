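# Niagara effect for the MicroLeaf phone emote that pops green, glowing lizard icons in a repeating eight-loop burst.
# Lizard_Icons launches sprites upward with a scale-pop curve and a particle light.
# Square_Glitches flickers mesh glitches at each spawn, while a delayed Despawn_Burst repeats them higher up as icons expire.
# Spawn groups select per-particle behavior, with brightness adjusted for time of day and front end, and glitch size by camera distance.
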NiagaraSystem /BRCosmetics/Animation/Game/MainPlayer/Emotes/MicroLeaf/FX/NS_Emote_MicroLeaf_Phone_Icons
dependencies {
 usedDataInterfaces [&NiagaraDataInterfaceCurve, &NiagaraDataInterfaceColorCurve, &NiagaraDataInterfaceSocketReader]
 usedDynamicInputs []
 usedModules [
  &self:SystemSpawnScript.SystemScriptSource.SystemScriptGraph.NiagaraNodeAssignment_0.SetVariables_#1
  &self:SystemSpawnScript.SystemScriptSource.SystemScriptGraph.NiagaraNodeAssignment_1.SetVariables_#2
  &self:SystemSpawnScript.SystemScriptSource.SystemScriptGraph.NiagaraNodeAssignment_2.SetVariables_#3
  &SystemState
 ]
}
emitters {
 Despawn_Burst {
  emitterUpdateScript [
   &EmitterState(
    Delay First Loop Only=true
    Inactive Response=::NewEnumerator1
    Life Cycle Mode=::NewEnumerator1
    Loop Behavior=::NewEnumerator2
    Loop Count=8
    Loop Delay=0.6
    Loop Duration=0.5
    UseLoopDelay=true
   )
   &SpawnBurst_Instantaneous(Spawn Count=2, Spawn Group=9, Spawn Time=0.1, Use Spawn Probability=false, UseLoopCountLimit=false) {module=SpawnBurst_Instantaneous009}
   &SpawnBurst_Instantaneous(Spawn Count=2, Spawn Group=9, Spawn Time=0.1, Use Spawn Probability=false, UseLoopCountLimit=false) {module=SpawnBurst_Instantaneous010}
   &SpawnBurst_Instantaneous(Spawn Count=15, Spawn Group=8, Spawn Time=0.1, Use Spawn Probability=false, UseLoopCountLimit=false) {module=SpawnBurst_Instantaneous008}
  ]
  particleSpawnScript [
   /SetVariables(Particles.VisibilityTag=&SpawnGroup()) {isSetParameters=true}
   /SetVariables(Particles.Random Range:NiagaraFloat=&RandomRangeFloat(Minimum=0.2, Maximum=0.8)) {isSetParameters=true}
   &InitializeParticle(
    Color Mode=::NewEnumerator0
    Lifetime=&MergeHelper_FloatSwitch(
     0=&RandomRangeFloat(Minimum=0.15, Maximum=0.2)
     1=&RandomRangeFloat(Minimum=0.12, Maximum=0.18)
     2=?
     3=?
     6=&RandomRangeFloat(Minimum=0.55, Maximum=0.55)
     7=&RandomRangeFloat(Minimum=0.5, Maximum=0.62)
     8=&RandomRangeFloat(Minimum=0.15, Maximum=0.2)
     9=&RandomRangeFloat(Minimum=0.12, Maximum=0.18)
     Index=&SpawnGroup()
     Num=::NewEnumerator0
    )
    Position Offset=<0 22 188>
    UsePositionOffset=true
    Write Material Random=true
   )
   &CameraOffset(Camera Offset Amount=&MergeHelper_FloatSwitch(0=15, 1=15, 2=?, 3=?, Index=&SpawnGroup(), Num=::NewEnumerator0))
   &InitialMeshOrientation(
    Orientation Axis=&MergeHelper_VectorSwitch(
     0=<0 0 0>
     1=<0 0 0>
     2=<0 0 0>
     3=<0 0 1>
     4=<0 0 0>
     5=<0 0 0>
     6=<0 0 0>
     7=<0 0 0>
     8=<0 0 0>
     9=<0 0 0>
     Index=&SpawnGroup()
     Num=::NewEnumerator8
    )
    Orientation Vector=&MergeHelper_VectorSwitch(0=<0 0 0>, 1=<0 0 0>, 2=?, 3=?, Index=&SpawnGroup(), Num=::NewEnumerator0)
    Rotation=&MergeHelper_VectorSwitch(
     0=<0 0 0>
     1=<0 0 0>
     2=<0 0 0>
     3=<0 0.25 0>
     4=<0 0 0>
     5=<0 0 0>
     6=<0 0 0>
     7=<0 0 0>
     8=<0 0 0>
     9=<0 0 0>
     Index=&SpawnGroup()
     Num=::NewEnumerator8
    )
    Use Orientation Vector=true
    Use Rotation Vector=true
   )
   &InitialMeshOrientation(
    Orientation Axis=<0 0 0>
    Orientation Vector=<0 0 0>
    Rotation=&MergeHelper_VectorSwitch(
     0=<0 0 0>
     1=<0 0 0>
     2=?
     3=&RandomRangeVector(Minimum=<0 0 0>, Maximum=<1 0 0>)
     Index=&SpawnGroup()
     Num=::NewEnumerator0
    )
    Use Orientation Vector=false
    Use Rotation Vector=true
   ) {module=InitialMeshOrientation001}
   /SetVariables(
    Particles.Long Glitch Dynam 1:Vector4f=&MakeVector4(X=0, Y=0, Z=0, W=&RandomRangeFloat(Minimum=-1, Maximum=0.1))
    Particles.Long Glitch Dynam 2:Vector4f=&MakeVector4(X=&RandomRangeFloat(Minimum=-100, Maximum=100), Y=0, Z=0, W=0)
    Particles.Long Glitch Size:Vector3f=&RandomRangeVector(Minimum=<0.25 2.5 0.25>, Maximum=<0.5 5 1>)
   ) {isSetParameters=true}
   /SetVariables(
    Particles.Square Glitch Dynam 1:Vector4f=&MakeVector4(X=0, Y=0, Z=0, W=&RandomRangeFloat(Minimum=-1, Maximum=0.1))
    Particles.Square Glitch Dynam 2:Vector4f=&MakeVector4(X=&RandomRangeFloat(Minimum=-100, Maximum=100), Y=0, Z=0, W=0)
    Particles.Square Glitch Size:Vector3f=&RandomRangeVector(Minimum=<1 1 2>, Maximum=<2.5 2.5 3>)
   ) {isSetParameters=true}
   &ShapeLocation(
    MaskSpawnGroup=false
    Non Uniform Scale=<0.5 0.5 0.8>
    Override Seed=false
    Sphere Radius=&MergeHelper_FloatSwitch(0=0, 1=0, 2=0, 3=0, 4=0, 5=0, 6=3, 7=3, 8=15, 9=10, Index=&SpawnGroup(), Num=::NewEnumerator8)
    UseOwnerScale=true
   )
   &AddVelocity(
    Velocity=&MergeHelper_VectorSwitch(
     0=<0 0 0>
     1=<0 0 0>
     2=<0 0 0>
     3=<0 0 0>
     4=<0 0 0>
     5=<0 0 0>
     6=$Particles.ShapeLocation.ShapeNormal
     7=$Particles.ShapeLocation.ShapeNormal
     8=<0 0 0.2>
     9=<0 0 0.2>
     Index=&SpawnGroup()
     Num=::NewEnumerator8
    )
    Velocity Speed Scale=&RandomRangeFloat(Minimum=120, Maximum=220)
   )
   &ApplyInitialForces()
  ]
  particleUpdateScript [
   &ParticleState()
   &UpdateMeshOrientation(
    Rotation Rate=&MaskFloat(Mask Source=0, Mask Parameter=?, Mask=2, Masked Float=0.25, Passthrough Float=0)
    Rotation Vector=&MergeHelper_VectorSwitch(
     0=<0 0 0>
     1=<0 0 0>
     2=<0 0 0>
     3=&RandomRangeVector(Minimum=<0 0 -1>, Maximum=<0 0 1>)
     4=<0 0 0>
     5=<0 0 0>
     6=<0 0 0>
     7=<0 0 0>
     8=<0 0 0>
     9=<0 0 0>
     Index=&SpawnGroup()
     Num=::NewEnumerator8
    )
   )
   /SetVariables(
    Particles.Square Glitch Color Update:LinearColor=&RandomRangeLinearColor(Minimum=<rgba 0.2535 1 0.2794 1>, Maximum=<rgba 0 1 0.1992 1>)
    Particles.Square Glitch DynParam 1:Vector4f=&MakeVector4(
     X=0
     Y=&RandomRangeFloat(Minimum=-1, Maximum=1)
     Z=&FloatFromCurve(FloatCurve=DataInterface() {propertyValues {Curve {keys [
          <at 0 2 RCIM_Cubic>
          {arriveTangent=-6.2285, interpMode=RCIM_Cubic, leaveTangent=-6.2285, tangentMode=RCTM_User, time=0.5612, value=1.0841}
          {arriveTangent=-0.8021, interpMode=RCIM_Cubic, leaveTangent=-0.8021, tangentMode=RCTM_User, time=1, value=0}
         ]}}})
     W=0
    )
    Particles.Square Glitch DynParam 2:Vector4f=&MakeVector4(
     X=&RandomRangeFloat(Minimum=-1, Maximum=1)
     Y=0
     Z=0
     W=&FloatFromCurve(FloatCurve=DataInterface() {propertyValues {Curve {keys [
          {arriveTangent=4.1785, interpMode=RCIM_Cubic, leaveTangent=4.1785, tangentMode=RCTM_User, time=0, value=0}
          {arriveTangent=0.9995, interpMode=RCIM_Cubic, leaveTangent=0.9995, tangentMode=RCTM_User, time=0.2223, value=0.6825}
          {arriveTangent=-0.0405, interpMode=RCIM_Cubic, leaveTangent=-0.0405, tangentMode=RCTM_User, time=1, value=1}
         ]}}})
    )
    Particles.Square Glitch Scale Update:Vector3f=&Multiply_Vector(
     A=$Particles.Initial.Square Glitch Size
     B=&Multiply_VectorByFloat(
      Vector=<0.25 0.25 0.25>
      Float=&Multiply_Float(
       A=0.08
       B=&FloatFromCurve(
        FloatCurve=DataInterface() {propertyValues {Curve {keys [
            {arriveTangent=5.4245, interpMode=RCIM_Cubic, leaveTangent=5.4245, tangentMode=RCTM_User, time=0, value=5}
            {arriveTangent=-3.9614, interpMode=RCIM_Cubic, leaveTangent=-3.9614, time=0.0963, value=8.5355}
            {arriveTangent=-22.2662, interpMode=RCIM_Cubic, leaveTangent=-22.2662, time=0.2702, value=3.9295}
            <at 0.3 4>
            <at 0.4 11.5>
            {arriveTangent=0.1532, interpMode=RCIM_Cubic, leaveTangent=0.1532, tangentMode=RCTM_User, time=0.5, value=2.875}
            <at 0.6 3>
            <at 0.8 2.5>
            <at 0.9 2>
            {arriveTangent=-0.0405, interpMode=RCIM_Cubic, leaveTangent=-0.0405, tangentMode=RCTM_User, time=1, value=0.5}
           ]}}}
        Scale Curve=&Multiply_Float(
         A=0.3
         B=&ClampFloat(
          Float=&FN_LOD_Distance_Remap(Start Distance=200, End Distance=400, Minimum Output=0.5, Maximum Output=1)
          Min=0.5
         )
        )
       )
      )
     )
    )
   ) {isSetParameters=true}
   /SetVariables(
    Particles.Long Glitch Color Update:LinearColor=&RandomRangeLinearColor(Minimum=<rgba 0.8407 1 0.4066 1>, Maximum=<rgba 0 1 0.1992 1>)
    Particles.Long Glitch Dynam 1:Vector4f=&MakeVector4(
     X=0
     Y=&RandomRangeFloat(Minimum=-1, Maximum=1)
     Z=&FloatFromCurve(FloatCurve=DataInterface() {propertyValues {Curve {keys [
          <at 0 2 RCIM_Cubic>
          {arriveTangent=-6.2285, interpMode=RCIM_Cubic, leaveTangent=-6.2285, tangentMode=RCTM_User, time=0.5612, value=1.0841}
          {arriveTangent=-0.8021, interpMode=RCIM_Cubic, leaveTangent=-0.8021, tangentMode=RCTM_User, time=1, value=0}
         ]}}})
     W=0
    )
    Particles.Long Glitch Dynam 2:Vector4f=&MakeVector4(
     X=0
     Y=0
     Z=0
     W=&FloatFromCurve(FloatCurve=DataInterface() {propertyValues {Curve {keys [
          {arriveTangent=4.1785, interpMode=RCIM_Cubic, leaveTangent=4.1785, tangentMode=RCTM_User, time=0, value=0}
          {arriveTangent=0.9995, interpMode=RCIM_Cubic, leaveTangent=0.9995, tangentMode=RCTM_User, time=0.2223, value=0.6825}
          {arriveTangent=-0.0405, interpMode=RCIM_Cubic, leaveTangent=-0.0405, tangentMode=RCTM_User, time=1, value=1}
         ]}}})
    )
    Particles.Long Glitch Scale:Vector3f=&Multiply_Vector(
     A=$Particles.Initial.Long Glitch Size
     B=&Multiply_VectorByFloat(
      Vector=<0.5 0.5 0.2>
      Float=&Multiply_Float(
       A=0.08
       B=&FloatFromCurve(
        FloatCurve=DataInterface() {propertyValues {Curve {keys [
            <at 0.0024 4.7657 RCIM_Cubic>
            {arriveTangent=14.525, interpMode=RCIM_Cubic, leaveTangent=14.525, tangentMode=RCTM_Break, time=0.0625, value=7.4517}
            {arriveTangent=-14.3776, interpMode=RCIM_Cubic, leaveTangent=-14.3776, tangentMode=RCTM_Break, time=0.15, value=6.2624}
            {arriveTangent=0.1532, interpMode=RCIM_Cubic, leaveTangent=0.1532, tangentMode=RCTM_Break, time=0.5, value=2.875}
            {interpMode=RCIM_Cubic, tangentMode=RCTM_Break, time=0.8185, value=2.3106}
            {arriveTangent=-0.0405, interpMode=RCIM_Cubic, leaveTangent=-0.0405, tangentMode=RCTM_Break, time=0.9185, value=0.8106}
           ]}}}
        Scale Curve=&Multiply_Float(
         A=0.5
         B=&ClampFloat(
          Float=&FN_LOD_Distance_Remap(Start Distance=200, End Distance=400, Minimum Output=0.5, Maximum Output=1)
          Min=0.5
         )
        )
       )
      )
     )
    )
   ) {isSetParameters=true}
   /SetVariables(
    Particles.Soft glow SpriteSize:Vector2f=<15 15>
    Particles.soft glow color:LinearColor=&MakeLinearColorFromVectorAndFloat(
     "Vector (RGB)"=&MakeVectorFromLinearColorRGB(Color=<rgba 0.2679 1 0 1>)
     "Float (Alpha)"=&FloatFromCurve(
      FloatCurve=DataInterface() {propertyValues {Curve {
         defaultValue=1
         keys [
          {arriveTangent=9.0725, interpMode=RCIM_Cubic, leaveTangent=9.0725, tangentMode=RCTM_User, time=0, value=0}
          {arriveTangent=0.2227, interpMode=RCIM_Cubic, leaveTangent=0.2227, tangentMode=RCTM_User, time=0.1408, value=0.9847}
          {arriveTangent=-1.8564, interpMode=RCIM_Cubic, leaveTangent=-1.8564, tangentMode=RCTM_User, time=1, value=0}
         ]
        }}}
      Scale Curve=&FN_FrontEndFloatModifier(
       Gameplay=&FNTimeOfDayScalar(Night Scale=0.05, Morning Scale=0.02, Day Scale=0.03, Evening Scale=0.02)
       Front End=0.01
      )
     )
    )
   ) {isSetParameters=true}
   &Drag(
    Drag=&MergeHelper_FloatSwitch(
     0=0
     1=0
     2=0
     3=0
     4=0
     5=0
     6=&FloatFromCurve(
      FloatCurve=DataInterface() {propertyValues {Curve {keys [
          {arriveTangent=2.0771, interpMode=RCIM_Cubic, leaveTangent=2.0771, tangentMode=RCTM_User, time=0, value=0}
          {interpMode=RCIM_Cubic, tangentMode=RCTM_User, time=1, value=1}
         ]}}}
      Scale Curve=10
     )
     7=&FloatFromCurve(
      FloatCurve=DataInterface() {propertyValues {Curve {keys [
          {arriveTangent=2.0771, interpMode=RCIM_Cubic, leaveTangent=2.0771, tangentMode=RCTM_User, time=0, value=0}
          {interpMode=RCIM_Cubic, tangentMode=RCTM_User, time=1, value=1}
         ]}}}
      Scale Curve=10
     )
     Index=&SpawnGroup()
     Num=::NewEnumerator6
    )
    Rotational Drag:NiagaraFloat=&MergeHelper_FloatSwitch(
     0=0
     1=0
     2=0
     3=0
     4=0
     5=0
     6=&FloatFromCurve(
      FloatCurve=DataInterface() {propertyValues {Curve {keys [
          {arriveTangent=2.0771, interpMode=RCIM_Cubic, leaveTangent=2.0771, tangentMode=RCTM_User, time=0, value=0}
          {interpMode=RCIM_Cubic, tangentMode=RCTM_User, time=1, value=1}
         ]}}}
      Scale Curve=3
     )
     7=&FloatFromCurve(
      FloatCurve=DataInterface() {propertyValues {Curve {keys [
          {arriveTangent=2.0771, interpMode=RCIM_Cubic, leaveTangent=2.0771, tangentMode=RCTM_User, time=0, value=0}
          {interpMode=RCIM_Cubic, tangentMode=RCTM_User, time=1, value=1}
         ]}}}
      Scale Curve=3
     )
     Index=&SpawnGroup()
     Num=::NewEnumerator6
    )
    Use Linear Drag=true
    Use Rotational Drag=true
   )
   &SolveForcesAndVelocity(Clamp Velocity=false, Limit Acceleration=false)
  ]
  rendererClasses [&NiagaraMeshRendererProperties, &NiagaraSpriteRendererProperties]
  renderers [
   {rendererClass=&NiagaraMeshRendererProperties, rendererIndex=0}
   {rendererClass=&NiagaraMeshRendererProperties, rendererIndex=1}
   {rendererClass=&NiagaraSpriteRendererProperties, rendererIndex=2}
  ]
 }
 Lizard_Icons {
  emitterUpdateScript [
   &EmitterState(Life Cycle Mode=::NewEnumerator1, Loop Behavior=::NewEnumerator2, Loop Count=8, Loop Duration=0.5, UseLoopDelay=false)
   &SpawnBurst_Instantaneous(Spawn Group=1, Spawn Time=0.007, Use Spawn Probability=false, UseLoopCountLimit=false)
  ]
  particleSpawnScript [
   /SetVariables(Particles.VisibilityTag=&SpawnGroup()) {isSetParameters=true}
   &FN_SetSpawnGroup(Number of Spawn Groups=2)
   &InitializeParticle(
    Color Mode=::NewEnumerator0
    Lifetime=&MergeHelper_FloatSwitch(0=0.85, 1=0.85, 2=?, 3=?, Index=&SpawnGroup(), Num=::NewEnumerator0)
    Position Mode=::NewEnumerator1
    Sprite Rotation Angle=&RandomRangeFloat(Minimum=0, Maximum=1.03)
    Sprite Rotation Mode=::NewEnumerator0
    Sprite Size Mode=::NewEnumerator3
    Sprite UV Mode=::NewEnumerator1
    Uniform Sprite Size=25
    UsePositionOffset=false
    Write Material Random=true
   )
   &SocketLocation(SamplePosOffset=false, Socket Reader=$System.Socket Reader, Socket Sampling Mode=::NewEnumerator3, VelocitySpeedLimit=false)
   &AddVelocity(Velocity=<0 0 90>)
   &ApplyInitialForces()
   &InitialMeshOrientation(
    Mesh Axis To Orient=<0 -1 1>
    Mesh Orientation Mode=::NewEnumerator2
    Orientation Coordinate Space=::Local
    Orientation Vector=<0 -1 0.5>
    Use Rotation Vector=false
   )
   /SetVariables(Particles.Position=&AddVectorToPosition(Position=<0 0 0>, Vector=<0 22 135>)) {isSetParameters=true}
  ]
  particleUpdateScript [
   &ParticleState()
   &SpriteRotationRate(Rotation Rate=&RandomRangeFloat(Minimum=-100, Maximum=100))
   /SetVariables(Particles.Lizard Icon Size:Vector2f=&Vector2DFromFloat(Value=&FloatFromCurve(
      FloatCurve=DataInterface() {propertyValues {Curve {keys [
          <at 0 0 RCIM_Cubic>
          {arriveTangent=-0.0027, interpMode=RCIM_Cubic, leaveTangent=-0.0027, tangentMode=RCTM_User, time=0.1034, value=1.0763}
          {arriveTangent=-0.0813, interpMode=RCIM_Cubic, leaveTangent=-0.0813, tangentMode=RCTM_User, time=0.15, value=0.7971}
          {arriveTangent=-0.144, interpMode=RCIM_Cubic, leaveTangent=-0.144, tangentMode=RCTM_User, time=0.65, value=0.7921}
          {arriveTangent=-2.2046, interpMode=RCIM_Cubic, leaveTangent=-2.2046, tangentMode=RCTM_User, time=0.7051, value=1.632}
          {arriveTangent=-6.6509, interpMode=RCIM_Cubic, leaveTangent=-6.6509, time=0.8058, value=0.8113}
          <at 0.9505 0 RCIM_Cubic>
         ]}}}
      Scale Curve=&Multiply_Float(A=30, B=0.5)
     ))) {isSetParameters=true}
   &Drag(Drag=0.5, Use Linear Drag=true, Use Rotational Drag=true)
   &SolveForcesAndVelocity(Clamp Velocity=false, Limit Acceleration=false)
   &ScaleColor(
    Scale Alpha=&FN_FrontEndFloatModifier(
     Gameplay=&FNTimeOfDayScalar(Night Scale=3, Morning Scale=1, Day Scale=1.5, Evening Scale=1)
     Front End=1.5
    )
    ScaleA=true
    ScaleRGB=true
   ) {module=ScaleColor001}
   /SetVariables(
    Particles.Lizard Glow SpriteSize:Vector2f=&Vector2DFromFloat(Value=&FloatFromCurve(
      FloatCurve=DataInterface() {propertyValues {Curve {keys [
          <at 0 0 RCIM_Cubic>
          {arriveTangent=-0.0027, interpMode=RCIM_Cubic, leaveTangent=-0.0027, tangentMode=RCTM_User, time=0.1034, value=1.0763}
          {arriveTangent=-0.0813, interpMode=RCIM_Cubic, leaveTangent=-0.0813, tangentMode=RCTM_User, time=0.15, value=0.7971}
          {arriveTangent=-0.144, interpMode=RCIM_Cubic, leaveTangent=-0.144, tangentMode=RCTM_User, time=0.65, value=0.7921}
          {arriveTangent=-2.2046, interpMode=RCIM_Cubic, leaveTangent=-2.2046, tangentMode=RCTM_User, time=0.7051, value=1.632}
          {arriveTangent=-6.6509, interpMode=RCIM_Cubic, leaveTangent=-6.6509, time=0.8058, value=0.8113}
          <at 0.9505 0 RCIM_Cubic>
         ]}}}
      Scale Curve=&Multiply_Float(A=30, B=1)
     ))
    Particles.Lizard Scale hologram:Vector3f=&Multiply_VectorByFloat(
     Vector=<0.5 0.5 1>
     Float=&FloatFromCurve(
      FloatCurve=DataInterface() {propertyValues {Curve {defaultValue=1, keys [{tangentMode=RCTM_User, time=0, value=0}, {tangentMode=RCTM_User, time=0.4468, value=0.9812}, <at 1 0>]}}}
      Scale Curve=0.7
     )
    )
    Particles.Lizard hologram Position:NiagaraPosition=&AddVectorToPosition(Position=$Engine.Emitter.SimulationPosition, Vector=<0 26 117>)
    Particles.Lizard hologram color:LinearColor=&MakeLinearColorFromVectorAndFloat(
     "Vector (RGB)"=&MakeVectorFromLinearColorRGB(Color=&ColorFromCurve(ColorCurve=DataInterface() {propertyValues {
         AlphaCurve {keys [<at 0 1 RCIM_Cubic>,<at 1 0 RCIM_Cubic>]}
         BlueCurve {keys [<at 0 0.4796 RCIM_Cubic>]}
         GreenCurve {keys [<at 0 1 RCIM_Cubic>]}
         RedCurve {keys [<at 0 0.3597 RCIM_Cubic>]}
        }}))
     "Float (Alpha)"=&FloatFromCurve(
      FloatCurve=DataInterface() {propertyValues {Curve {defaultValue=1, keys [{tangentMode=RCTM_User, time=0, value=0}, {tangentMode=RCTM_User, time=0.05, value=1}]}}}
      Scale Curve=&FNTimeOfDayScalar(Night Scale=0.6, Morning Scale=1, Evening Scale=0.6)
     )
    )
    Particles.Lizard hologram glow orientation:Quat4f=<0 0.299 0 0>
   ) {isSetParameters=true}
   /SetVariables(
    Particles.Lizard Glow CameraOffset=-2
    Particles.Lizard Glow Color:LinearColor=&MakeLinearColorFromVectorAndFloat(
     "Vector (RGB)"=&MakeVectorFromLinearColorRGB(Color=&ColorFromCurve(ColorCurve=DataInterface() {propertyValues {
         AlphaCurve {keys [<at 0 0 RCIM_Cubic>,<at 0.0335 0.5>,<at 1 0.5 RCIM_Cubic>]}
         BlueCurve {keys [<at 0 0.0936 RCIM_Cubic>]}
         GreenCurve {keys [<at 0 0.5 RCIM_Cubic>]}
         RedCurve {keys [<at 0 0 RCIM_Cubic>]}
        }}))
     "Float (Alpha)"=&FN_FrontEndFloatModifier(
      Gameplay=&FloatFromCurve(
       FloatCurve=DataInterface() {propertyValues {Curve {defaultValue=1, keys [{tangentMode=RCTM_User, time=0, value=0}, {tangentMode=RCTM_User, time=0.05, value=1}]}}}
       Scale Curve=&FNTimeOfDayScalar(Night Scale=0.25, Morning Scale=0.32, Day Scale=0.25, Evening Scale=0.2)
      )
      Front End=0.2
     )
    )
    Particles.Lizard Glow SpriteSize:Vector2f=&Vector2DFromFloat(Value=&FloatFromCurve(
      FloatCurve=DataInterface() {propertyValues {Curve {keys [
          <at 0 0 RCIM_Cubic>
          {arriveTangent=-0.0027, interpMode=RCIM_Cubic, leaveTangent=-0.0027, tangentMode=RCTM_User, time=0.1034, value=1.0763}
          {arriveTangent=-0.0813, interpMode=RCIM_Cubic, leaveTangent=-0.0813, tangentMode=RCTM_User, time=0.15, value=0.7971}
          {arriveTangent=-0.144, interpMode=RCIM_Cubic, leaveTangent=-0.144, tangentMode=RCTM_User, time=0.65, value=0.7921}
          {arriveTangent=-2.2046, interpMode=RCIM_Cubic, leaveTangent=-2.2046, tangentMode=RCTM_User, time=0.7051, value=1.632}
          {arriveTangent=-6.6509, interpMode=RCIM_Cubic, leaveTangent=-6.6509, time=0.8058, value=0.8113}
          <at 0.9505 0 RCIM_Cubic>
         ]}}}
      Scale Curve=&Multiply_Float(A=30, B=1.1)
     ))
   ) {isSetParameters=true}
   &FN_Light_Attributes(
    Color=<rgba 0.6014 5 0 1>
    Create Color Attribute=true
    Create Exponent Attribute=true
    Create Position Attribute=true
    Create Radius Attribute=true
    Enable Particle Light Enable=true
    Enable Visibility Tag=true
    Exponent=5
    Light Enabled=true
    Position=&AddVectorToPosition(Position=$Engine.Emitter.SimulationPosition, Vector=<0 0 0>)
    Radius=25
    Scale Color=&FN_FrontEndFloatModifier(Gameplay=1, Front End=0.57)
   )
  ]
  rendererClasses [&NiagaraLightRendererProperties, &NiagaraSpriteRendererProperties]
  renderers [
   {rendererClass=&NiagaraLightRendererProperties, rendererIndex=0}
   {rendererClass=&NiagaraSpriteRendererProperties, rendererIndex=1}
   {rendererClass=&NiagaraSpriteRendererProperties, rendererIndex=2}
  ]
 }
 Square_Glitches {
  emitterUpdateScript [
   &EmitterState(
    Delay First Loop Only=true
    Inactive Response=::NewEnumerator1
    Life Cycle Mode=::NewEnumerator1
    Loop Behavior=::NewEnumerator2
    Loop Count=8
    Loop Duration=0.5
    UseLoopDelay=true
   )
   &SpawnBurst_Instantaneous(Spawn Count=7, Spawn Group=9, Spawn Time=0.005, Use Spawn Probability=false, UseLoopCountLimit=false) {module=SpawnBurst_Instantaneous009}
   &SpawnBurst_Instantaneous(Spawn Count=3, Spawn Group=9, Spawn Time=0.005, Use Spawn Probability=false, UseLoopCountLimit=false) {module=SpawnBurst_Instantaneous010}
   &SpawnBurst_Instantaneous(Spawn Count=15, Spawn Group=8, Spawn Time=0.005, Use Spawn Probability=false, UseLoopCountLimit=false) {module=SpawnBurst_Instantaneous008}
  ]
  particleSpawnScript [
   /SetVariables(Particles.VisibilityTag=&SpawnGroup()) {isSetParameters=true}
   /SetVariables(Particles.Random Range:NiagaraFloat=&RandomRangeFloat(Minimum=0.2, Maximum=0.8)) {isSetParameters=true}
   &InitializeParticle(
    Color Mode=::NewEnumerator0
    Lifetime=&MergeHelper_FloatSwitch(
     0=0.4
     1=0.27
     2=0.35
     3=0.35
     4=0.37
     5=0.5
     6=&RandomRangeFloat(Minimum=0.55, Maximum=0.55)
     7=&RandomRangeFloat(Minimum=0.5, Maximum=0.62)
     8=&RandomRangeFloat(Minimum=0.12, Maximum=0.18)
     9=&RandomRangeFloat(Minimum=0.12, Maximum=0.18)
     Index=&SpawnGroup()
     Num=::NewEnumerator8
    )
    Position Offset=<0 22 140>
    UsePositionOffset=true
    Write Material Random=true
   ) {module=InitializeParticle001}
   &CameraOffset(Camera Offset Amount=&MergeHelper_FloatSwitch(0=0, 1=2, 2=2, 3=2, 4=2, 5=2, 6=2, 7=2, 8=15, 9=15, Index=&SpawnGroup(), Num=::NewEnumerator8))
   &InitialMeshOrientation(
    Orientation Axis=&MergeHelper_VectorSwitch(
     0=<0 0 0>
     1=<0 0 0>
     2=<0 0 0>
     3=<0 0 1>
     4=<0 0 0>
     5=<0 0 0>
     6=<0 0 0>
     7=<0 0 0>
     8=<0 0 0>
     9=<0 0 0>
     Index=&SpawnGroup()
     Num=::NewEnumerator8
    )
    Orientation Vector=&MergeHelper_VectorSwitch(
     0=<0 0 0>
     1=<0 0 0>
     2=<0 0 0>
     3=<0 0 1>
     4=<0 0 0>
     5=<0 0 0>
     6=<0 0 0>
     7=<0 0 0>
     8=<0 0 0>
     9=<0 0 0>
     Index=&SpawnGroup()
     Num=::NewEnumerator8
    )
    Rotation=&MergeHelper_VectorSwitch(
     0=<0 0 0>
     1=<0 0 0>
     2=<0 0 0>
     3=<0 0.25 0>
     4=<0 0 0>
     5=<0 0 0>
     6=<0 0 0>
     7=<0 0 0>
     8=<0 0 0>
     9=<0 0 0>
     Index=&SpawnGroup()
     Num=::NewEnumerator8
    )
    Use Orientation Vector=true
    Use Rotation Vector=true
   )
   &InitialMeshOrientation(
    Orientation Axis=<0 0 0>
    Orientation Vector=<0 0 0>
    Rotation=&MergeHelper_VectorSwitch(
     0=<0 0 0>
     1=<0 0 0>
     2=<0 0 0>
     3=&RandomRangeVector(Minimum=<0 0 0>, Maximum=<1 0 0>)
     4=<0 0 0>
     5=<0 0 0>
     6=<0 0 0>
     7=<0 0 0>
     8=<0 0 0>
     9=<0 0 0>
     Index=&SpawnGroup()
     Num=::NewEnumerator8
    )
    Use Orientation Vector=false
    Use Rotation Vector=true
   ) {module=InitialMeshOrientation001}
   /SetVariables(Particles.Long Glitch Size:Vector3f=&RandomRangeVector(Minimum=<0.25 2.5 0.25>, Maximum=<0.5 5 1>)) {isSetParameters=true}
   /SetVariables(
    Particles.Square Glitch Color:LinearColor=&RandomRangeLinearColor(Minimum=<rgba 0.2535 1 0.2794 1>, Maximum=<rgba 0 1 0.1992 1>)
    Particles.Square Glitch Size:Vector3f=&RandomRangeVector(Minimum=<1 1 2>, Maximum=<2.5 2.5 3>)
   ) {isSetParameters=true}
   &ShapeLocation(
    MaskSpawnGroup=false
    Non Uniform Scale=<0.65 0.65 1>
    Override Seed=false
    Sphere Radius=&MergeHelper_FloatSwitch(0=0, 1=0, 2=0, 3=0, 4=0, 5=0, 6=3, 7=3, 8=10, 9=10, Index=&SpawnGroup(), Num=::NewEnumerator8)
    UseOwnerScale=true
   )
   &ApplyInitialForces()
  ]
  particleUpdateScript [
   &ParticleState()
   /SetVariables(
    Particles.Square Glitch Color Update:LinearColor=&RandomRangeLinearColor(Minimum=<rgba 0.2535 1 0.2794 1>, Maximum=<rgba 0 1 0.1992 1>)
    Particles.Square Glitch DynParam 1:Vector4f=&MakeVector4(
     X=0
     Y=&RandomRangeFloat(Minimum=-1, Maximum=1)
     Z=&FloatFromCurve(FloatCurve=DataInterface() {propertyValues {Curve {keys [
          <at 0 2 RCIM_Cubic>
          {arriveTangent=-6.2285, interpMode=RCIM_Cubic, leaveTangent=-6.2285, tangentMode=RCTM_User, time=0.5612, value=1.0841}
          {arriveTangent=-0.8021, interpMode=RCIM_Cubic, leaveTangent=-0.8021, tangentMode=RCTM_User, time=1, value=0}
         ]}}})
     W=0
    )
    Particles.Square Glitch DynParam 2:Vector4f=&MakeVector4(
     X=&RandomRangeFloat(Minimum=-1, Maximum=1)
     Y=0
     Z=0
     W=&FloatFromCurve(FloatCurve=DataInterface() {propertyValues {Curve {keys [
          {arriveTangent=4.1785, interpMode=RCIM_Cubic, leaveTangent=4.1785, tangentMode=RCTM_User, time=0, value=0}
          {arriveTangent=0.9995, interpMode=RCIM_Cubic, leaveTangent=0.9995, tangentMode=RCTM_User, time=0.2223, value=0.6825}
          {arriveTangent=-0.0405, interpMode=RCIM_Cubic, leaveTangent=-0.0405, tangentMode=RCTM_User, time=1, value=1}
         ]}}})
    )
    Particles.Square Glitch Scale Update:Vector3f=&Multiply_Vector(
     A=$Particles.Initial.Square Glitch Size
     B=&Multiply_VectorByFloat(
      Vector=&RandomRangeVector(Minimum=<0.05 0.05 0.05>, Maximum=<0.2 0.2 0.2>)
      Float=&Multiply_Float(
       A=0.08
       B=&FloatFromCurve(
        FloatCurve=DataInterface() {propertyValues {Curve {keys [
            {arriveTangent=5.4245, interpMode=RCIM_Cubic, leaveTangent=5.4245, tangentMode=RCTM_User, time=0, value=5.7247}
            {arriveTangent=9.4279, interpMode=RCIM_Cubic, leaveTangent=9.4279, time=0.0625, value=10.3835}
            {arriveTangent=-31.0729, interpMode=RCIM_Cubic, leaveTangent=-31.0729, time=0.15, value=7.1389}
            {arriveTangent=-20.9259, interpMode=RCIM_Cubic, leaveTangent=-20.9259, time=0.2702, value=3.9295}
            <at 0.3 4>
            <at 0.3956 5.8879>
            {arriveTangent=0.1532, interpMode=RCIM_Cubic, leaveTangent=0.1532, tangentMode=RCTM_User, time=0.5, value=2.875}
            <at 0.8 2.5>
            <at 0.9 2>
            {arriveTangent=-0.0405, interpMode=RCIM_Cubic, leaveTangent=-0.0405, tangentMode=RCTM_User, time=1, value=0.5}
           ]}}}
        Scale Curve=&Multiply_Float(
         A=0.8
         B=&ClampFloat(
          Float=&FN_LOD_Distance_Remap(Start Distance=200, End Distance=400, Minimum Output=0.5, Maximum Output=1)
          Min=0.5
         )
        )
       )
      )
     )
    )
   ) {isSetParameters=true}
   /SetVariables(
    Particles.Long Glitch Color Update:LinearColor=&RandomRangeLinearColor(Minimum=<rgba 0.8407 1 0.4066 1>, Maximum=<rgba 0 1 0.1992 1>)
    Particles.Long Glitch Dynam 1:Vector4f=&MakeVector4(
     X=0
     Y=&RandomRangeFloat(Minimum=-1, Maximum=1)
     Z=&FloatFromCurve(FloatCurve=DataInterface() {propertyValues {Curve {keys [
          <at 0 2 RCIM_Cubic>
          {arriveTangent=-6.2285, interpMode=RCIM_Cubic, leaveTangent=-6.2285, tangentMode=RCTM_User, time=0.5612, value=1.0841}
          {arriveTangent=-0.8021, interpMode=RCIM_Cubic, leaveTangent=-0.8021, tangentMode=RCTM_User, time=1, value=0}
         ]}}})
     W=0
    )
    Particles.Long Glitch Dynam 2:Vector4f=&MakeVector4(
     X=0
     Y=0
     Z=0
     W=&FloatFromCurve(FloatCurve=DataInterface() {propertyValues {Curve {keys [
          {arriveTangent=4.1785, interpMode=RCIM_Cubic, leaveTangent=4.1785, tangentMode=RCTM_User, time=0, value=0}
          {arriveTangent=0.9995, interpMode=RCIM_Cubic, leaveTangent=0.9995, tangentMode=RCTM_User, time=0.2223, value=0.6825}
          {arriveTangent=-0.0405, interpMode=RCIM_Cubic, leaveTangent=-0.0405, tangentMode=RCTM_User, time=1, value=1}
         ]}}})
    )
    Particles.Long Glitch Scale:Vector3f=&Multiply_Vector(
     A=$Particles.Initial.Long Glitch Size
     B=&Multiply_VectorByFloat(
      Vector=<0.5 0.5 0.2>
      Float=&Multiply_Float(
       A=0.08
       B=&FloatFromCurve(
        FloatCurve=DataInterface() {propertyValues {Curve {keys [
            <at 0.0024 4.7657 RCIM_Cubic>
            {arriveTangent=14.525, interpMode=RCIM_Cubic, leaveTangent=14.525, tangentMode=RCTM_Break, time=0.0625, value=7.4517}
            {arriveTangent=-14.3776, interpMode=RCIM_Cubic, leaveTangent=-14.3776, tangentMode=RCTM_Break, time=0.15, value=6.2624}
            {arriveTangent=0.1532, interpMode=RCIM_Cubic, leaveTangent=0.1532, tangentMode=RCTM_Break, time=0.5, value=2.875}
            {interpMode=RCIM_Cubic, tangentMode=RCTM_Break, time=0.8185, value=2.3106}
            {arriveTangent=-0.0405, interpMode=RCIM_Cubic, leaveTangent=-0.0405, tangentMode=RCTM_Break, time=0.9185, value=0.8106}
           ]}}}
        Scale Curve=&Multiply_Float(
         A=0.5
         B=&ClampFloat(
          Float=&FN_LOD_Distance_Remap(Start Distance=200, End Distance=400, Minimum Output=0.5, Maximum Output=1)
          Min=0.5
         )
        )
       )
      )
     )
    )
   ) {isSetParameters=true}
   &Drag(
    Drag=&MergeHelper_FloatSwitch(
     0=0
     1=0
     2=0
     3=0
     4=0
     5=0
     6=&FloatFromCurve(
      FloatCurve=DataInterface() {propertyValues {Curve {keys [
          {arriveTangent=2.0771, interpMode=RCIM_Cubic, leaveTangent=2.0771, tangentMode=RCTM_User, time=0, value=0}
          {interpMode=RCIM_Cubic, tangentMode=RCTM_User, time=1, value=1}
         ]}}}
      Scale Curve=10
     )
     7=&FloatFromCurve(
      FloatCurve=DataInterface() {propertyValues {Curve {keys [
          {arriveTangent=2.0771, interpMode=RCIM_Cubic, leaveTangent=2.0771, tangentMode=RCTM_User, time=0, value=0}
          {interpMode=RCIM_Cubic, tangentMode=RCTM_User, time=1, value=1}
         ]}}}
      Scale Curve=10
     )
     Index=&SpawnGroup()
     Num=::NewEnumerator6
    )
    Rotational Drag:NiagaraFloat=&MergeHelper_FloatSwitch(
     0=0
     1=0
     2=0
     3=0
     4=0
     5=0
     6=&FloatFromCurve(
      FloatCurve=DataInterface() {propertyValues {Curve {keys [
          {arriveTangent=2.0771, interpMode=RCIM_Cubic, leaveTangent=2.0771, tangentMode=RCTM_User, time=0, value=0}
          {interpMode=RCIM_Cubic, tangentMode=RCTM_User, time=1, value=1}
         ]}}}
      Scale Curve=3
     )
     7=&FloatFromCurve(
      FloatCurve=DataInterface() {propertyValues {Curve {keys [
          {arriveTangent=2.0771, interpMode=RCIM_Cubic, leaveTangent=2.0771, tangentMode=RCTM_User, time=0, value=0}
          {interpMode=RCIM_Cubic, tangentMode=RCTM_User, time=1, value=1}
         ]}}}
      Scale Curve=3
     )
     Index=&SpawnGroup()
     Num=::NewEnumerator6
    )
    Use Linear Drag=true
    Use Rotational Drag=true
   )
   &SolveForcesAndVelocity(Clamp Velocity=false, Limit Acceleration=false)
   /SetVariables(
    Particles.Soft glow SpriteSize:Vector2f=<7 7>
    Particles.soft glow color:LinearColor=&MakeLinearColorFromVectorAndFloat(
     "Vector (RGB)"=&MakeVectorFromLinearColorRGB(Color=<rgba 0.2679 1 0 1>)
     "Float (Alpha)"=&FloatFromCurve(
      FloatCurve=DataInterface() {propertyValues {Curve {
         defaultValue=1
         keys [
          {arriveTangent=9.0725, interpMode=RCIM_Cubic, leaveTangent=9.0725, tangentMode=RCTM_User, time=0, value=0}
          {arriveTangent=0.2227, interpMode=RCIM_Cubic, leaveTangent=0.2227, tangentMode=RCTM_User, time=0.1408, value=0.9847}
          {arriveTangent=-1.8564, interpMode=RCIM_Cubic, leaveTangent=-1.8564, tangentMode=RCTM_User, time=1, value=0}
         ]
        }}}
      Scale Curve=&FN_FrontEndFloatModifier(
       Gameplay=&FNTimeOfDayScalar(Night Scale=0.05, Morning Scale=0.02, Day Scale=0.03, Evening Scale=0.02)
       Front End=0.01
      )
     )
    )
   ) {isSetParameters=true}
  ]
  rendererClasses [&NiagaraMeshRendererProperties, &NiagaraSpriteRendererProperties]
  renderers [
   {rendererClass=&NiagaraMeshRendererProperties, rendererIndex=0}
   {rendererClass=&NiagaraMeshRendererProperties, rendererIndex=1}
   {rendererClass=&NiagaraSpriteRendererProperties, rendererIndex=2}
  ]
 }
}
summary {
 systemName=NS_Emote_MicroLeaf_Phone_Icons
 emitters [
  {emitterName=Lizard_Icons, rendererClasses [&NiagaraLightRendererProperties, &NiagaraSpriteRendererProperties]}
  {
   emitterName=Square_Glitches
   rendererClasses [&NiagaraMeshRendererProperties, &NiagaraSpriteRendererProperties]
  }
  {emitterName=Despawn_Burst, rendererClasses [&NiagaraMeshRendererProperties, &NiagaraSpriteRendererProperties]}
 ]
 userVariables [
  User.bIsFrontEnd {type=NiagaraBool}
  User.bIsFrontEndPreview {type=NiagaraBool}
  User.Is_Hero {type=NiagaraFloat, defaultValue=0.5, description=""}
  User.NormalizedNotifyProgress {type=NiagaraFloat, defaultValue=0, description=""}
 ]
}
