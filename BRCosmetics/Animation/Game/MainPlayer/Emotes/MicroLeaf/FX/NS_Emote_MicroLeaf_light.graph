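# Niagara light effect for the MicroLeaf emote, with glow meshes, a phone-screen sprite, a light and a hand motion-blur ribbon.
# Minimal bursts particles via User.ForceSpawn, driving colors with waveform flicker, time-of-day scaling and mostly Light_Intensity, while on_true blends ramp-in toward steady values.
# Platform switches adjust mesh offsets and scale for Mobile; a disabled duplicate, labeled as having mobile issues, lacks them.
# motion_blur_hand builds a fading AnimTrail from skeletal mesh sampling.
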NiagaraSystem /BRCosmetics/Animation/Game/MainPlayer/Emotes/MicroLeaf/FX/NS_Emote_MicroLeaf_light.NS_Emote_MicroLeaf_Light
dependencies {
 usedDataInterfaces [
  &NiagaraDataInterfaceCurve
  &NiagaraDataInterfaceColorCurve
  &NiagaraDataInterfacePlatformSet
  &NiagaraDataInterfaceSocketReader
  &NiagaraDataInterfaceSkeletalMesh
 ]
 usedDynamicInputs []
 usedModules [
  &self:SystemSpawnScript.SystemScriptSource.SystemScriptGraph.NiagaraNodeAssignment_2.SetVariables_#1
  &self:SystemSpawnScript.SystemScriptSource.SystemScriptGraph.NiagaraNodeAssignment_3.SetVariables_#2
  &SystemState
 ]
}
emitters {
 Minimal {
  emitterSpawnScript [/SetVariables() {isSetParameters=true}]
  emitterUpdateScript [
   &EmitterState(Inactive Response=::NewEnumerator1, Life Cycle Mode=::NewEnumerator1, Loop Duration=10.75, UseLoopDelay=false)
   &SpawnBurst_Instantaneous(
    Spawn Count=&Multiply_Int(A=1, B=&MakeIntFromFloat(Float=$User.ForceSpawn))
    Use Spawn Probability=false
    UseLoopCountLimit=false
   )
  ]
  particleSpawnScript [
   /SetVariables(Particles.VisibilityTag=&SpawnGroup()) {isSetParameters=true, module=SetVariables_#3}
   &InitializeParticle(Color Mode=::NewEnumerator0, Lifetime=10.75, Position Mode=::NewEnumerator1, UsePositionOffset=false, Write Material Random=true)
   &InitialMeshOrientation(
    Mesh Axis To Orient=<0 0 -1>
    Mesh Orientation Mode=::NewEnumerator2
    Orientation Coordinate Space=::Local
    Orientation Vector=&MergeHelper_VectorSwitch(0=<-0.1777 15 0.1184>, 1=<0 0 0>, 2=?, 3=?, Index=&SpawnGroup(), Num=::NewEnumerator0)
    Use Rotation Vector=false
   )
   &Color()
  ]
  particleUpdateScript [
   &ParticleState()
   /SetVariables(
    Particles.Core Green Color:LinearColor=&LerpLinearColors(
     StartColor=&MakeLinearColorFromVectorAndFloat(
      "Vector (RGB)"=&MakeVectorFromLinearColorRGB(Color=&MakeLinearColorFromVectorAndFloat(
        "Vector (RGB)"=&MakeVectorFromLinearColorRGB(Color=<rgba 0.3433 1 0 1>)
        "Float (Alpha)"=&FloatFromCurve(
         FloatCurve=DataInterface() {propertyValues {Curve {
            defaultValue=1
            keys [{interpMode=RCIM_Cubic, tangentMode=RCTM_Break, time=0, value=0}, {interpMode=RCIM_Cubic, tangentMode=RCTM_Break, time=0.05, value=1}]
           }}}
         Scale Curve=&Waveform(
          Global Input Position=1
          Global Amplitude Scale=&Multiply_Float(
           A=&FNTimeOfDayScalar(Night Scale=0.56, Morning Scale=0.45, Day Scale=0.56, Evening Scale=0.45)
           B=$User.Light_Intensity
          )
          "Waveform [1]"=::NewEnumerator0
          "[1] Frequency"=5
          "[1] Amplitude Min/Max"=<0.8 1>
          "[1] Scale"=0.5
         )
        )
       ))
      "Float (Alpha)"=1
     )
     EndColor=&Multiply_LinearColorByFloat(
      Linear Color=<rgba 1.0299 3 0 1>
      Float=&Waveform(
       Global Input Position=1
       Global Amplitude Scale=&Multiply_Float(
        A=&FNTimeOfDayScalar(Night Scale=0.7, Morning Scale=0.55, Day Scale=0.7, Evening Scale=0.55)
        B=$User.Light_Intensity
       )
       "Waveform [1]"=::NewEnumerator0
       "[1] Frequency"=5
       "[1] Amplitude Min/Max"=<0.8 1>
       "[1] Scale"=0.5
      )
     )
     LerpFactor=$User.on_true
    )
    Particles.Core Light Color:LinearColor=&LerpLinearColors(
     StartColor=&MakeLinearColorFromVectorAndFloat(
      "Vector (RGB)"=&MakeVectorFromLinearColorRGB(Color=&MakeLinearColorFromVectorAndFloat(
        "Vector (RGB)"=&MakeVectorFromLinearColorRGB(Color=<rgba 0.139 0.4339 5 1>)
        "Float (Alpha)"=&FloatFromCurve(
         FloatCurve=DataInterface() {propertyValues {Curve {
            defaultValue=1
            keys [{interpMode=RCIM_Cubic, tangentMode=RCTM_Break, time=0, value=0}, {interpMode=RCIM_Cubic, tangentMode=RCTM_Break, time=0.05, value=1}]
           }}}
         Scale Curve=&Waveform(
          Global Input Position=1
          Global Amplitude Scale=&Multiply_Float(
           A=&FNTimeOfDayScalar(Night Scale=0.7, Morning Scale=0.55, Day Scale=0.7, Evening Scale=0.55)
           B=$User.Light_Intensity
          )
          "Waveform [1]"=::NewEnumerator0
          "[1] Frequency"=5
          "[1] Amplitude Min/Max"=<0.8 1>
          "[1] Scale"=0.5
         )
        )
       ))
      "Float (Alpha)"=1
     )
     EndColor=&Multiply_LinearColorByFloat(
      Linear Color=<rgba 0.1668 0.5207 6 1>
      Float=&Waveform(
       Global Input Position=1
       Global Amplitude Scale=&Multiply_Float(
        A=&FNTimeOfDayScalar(Night Scale=0.7, Morning Scale=0.55, Day Scale=0.7, Evening Scale=0.55)
        B=$User.Light_Intensity
       )
       "Waveform [1]"=::NewEnumerator0
       "[1] Frequency"=5
       "[1] Amplitude Min/Max"=<0.8 1>
       "[1] Scale"=0.5
      )
     )
     LerpFactor=$User.on_true
    )
    Particles.Core Light Position:NiagaraPosition=&AddVectorToPosition(
     Position=$Engine.Emitter.SimulationPosition
     Vector=&MakeVector(
      X=-12
      Y=&FN_PickFloatDependOnPlatform(
       TargetPlatform=DataInterface() {propertyValues {Platforms {qualityLevelMask=-1, deviceProfileStates [{profileName=Mobile, qualityLevelMask=0, setQualityLevelMask=4294967295}]}}}
       TargetPlatformActive=-1.3
       TargetPlatformNoActive=0.2
      )
      Z=-14
     )
    )
    Particles.Core Light Scale:Vector3f=/Niagara/DynamicInputs/Math/LerpMultipleVectors(
     Alpha:NiagaraFloat=$User.on_true
     V 1:Vector3f=&Multiply_Vector(
      A=&MakeVector(
       X=0.73
       Y=0.85
       Z=&FN_PickFloatDependOnPlatform(
        TargetPlatform=DataInterface() {propertyValues {Platforms {qualityLevelMask=-1, deviceProfileStates [{profileName=Mobile, qualityLevelMask=0, setQualityLevelMask=4294967295}]}}}
        TargetPlatformActive=0.7
        TargetPlatformNoActive=0.4
       )
      )
      B=&VectorFromFloat(Value=&FloatFromCurve(FloatCurve=DataInterface() {propertyValues {Curve {
           defaultValue=1
           keys [{interpMode=RCIM_Cubic, tangentMode=RCTM_Break, time=0, value=0}, {interpMode=RCIM_Cubic, tangentMode=RCTM_Break, time=0.02, value=1}]
          }}}))
     )
     V 2:Vector3f=&MakeVector(
      X=0.73
      Y=0.85
      Z=&FN_PickFloatDependOnPlatform(
       TargetPlatform=DataInterface() {propertyValues {Platforms {qualityLevelMask=-1, deviceProfileStates [{profileName=Mobile, qualityLevelMask=0, setQualityLevelMask=4294967295}]}}}
       TargetPlatformActive=0.7
       TargetPlatformNoActive=0.4
      )
     )
     V 3:Vector3f=?
     V 4:Vector3f=?
     V 5:Vector3f=?
     Vector Count:ENiagara_MultipleLerpCount=::NewEnumerator3
    )
   ) {isSetParameters=true, module=SetVariables_#4}
   /SetVariables(
    Particles.Outer Glow Color:LinearColor=&LerpLinearColors(
     StartColor=&MakeLinearColorFromVectorAndFloat(
      "Vector (RGB)"=&MakeVectorFromLinearColorRGB(Color=&FNTimeOfDayColor(
        Night Color=<rgba 0.0688 0.1859 2 1>
        Morning Color=<rgba 0.0278 0.0868 1 1>
        Day Color=<rgba 0.0278 0.0868 1 1>
        Evening Color=<rgba 0.0278 0.0868 1 1>
       ))
      "Float (Alpha)"=&FloatFromCurve(
       FloatCurve=DataInterface() {propertyValues {Curve {
          defaultValue=1
          keys [{interpMode=RCIM_Cubic, tangentMode=RCTM_Break, time=0, value=1}, {interpMode=RCIM_Cubic, tangentMode=RCTM_Break, time=0.05, value=1}]
         }}}
       Scale Curve=&Waveform(
        Global Amplitude Scale=&FNTimeOfDayScalar(Night Scale=0.53, Morning Scale=0.35, Day Scale=0.67, Evening Scale=0.35)
        "Waveform [1]"=::NewEnumerator0
        "[1] Frequency"=1
        "[1] Amplitude Min/Max"=<0.85 1>
       )
      )
     )
     EndColor=&MakeLinearColorFromVectorAndFloat(
      "Vector (RGB)"=&MakeVectorFromLinearColorRGB(Color=&FNTimeOfDayColor(
        Night Color=<rgba 0.0688 0.1859 2 1>
        Morning Color=<rgba 0.0278 0.0868 1 1>
        Day Color=<rgba 0.0278 0.0868 1 1>
        Evening Color=<rgba 0.0278 0.0868 1 1>
       ))
      "Float (Alpha)"=&Waveform(
       Global Amplitude Scale=&FNTimeOfDayScalar(Night Scale=0.53, Morning Scale=0.35, Day Scale=0.67, Evening Scale=0.35)
       "Waveform [1]"=::NewEnumerator0
       "[1] Frequency"=1
       "[1] Amplitude Min/Max"=<0.85 1>
      )
     )
     LerpFactor=$User.on_true
    )
    Particles.Outer Glow Position:NiagaraPosition=&AddVectorToPosition(
     Position=$Engine.Emitter.SimulationPosition
     Vector=&MakeVector(
      X=-12
      Y=&FN_PickFloatDependOnPlatform(
       TargetPlatform=DataInterface() {propertyValues {Platforms {qualityLevelMask=-1, deviceProfileStates [{profileName=Mobile, qualityLevelMask=0, setQualityLevelMask=4294967295}]}}}
       TargetPlatformActive=-1.3
       TargetPlatformNoActive=0
      )
      Z=-14
     )
    )
    Particles.Outer Glow Scale:Vector3f=&LerpMultipleVectors(
     Alpha:NiagaraFloat=$User.on_true
     V 1:Vector3f=&Multiply_Vector(
      A=&MakeVector(
       X=0.68
       Y=0.85
       Z=&FN_PickFloatDependOnPlatform(
        TargetPlatform=DataInterface() {propertyValues {Platforms {qualityLevelMask=-1, deviceProfileStates [{profileName=Mobile, qualityLevelMask=0, setQualityLevelMask=4294967295}]}}}
        TargetPlatformActive=0.7
        TargetPlatformNoActive=0.4
       )
      )
      B=&VectorFromFloat(Value=&FloatFromCurve(FloatCurve=DataInterface() {propertyValues {Curve {
           defaultValue=1
           keys [{interpMode=RCIM_Cubic, tangentMode=RCTM_Break, time=0, value=0}, {interpMode=RCIM_Cubic, tangentMode=RCTM_Break, time=0.05, value=1}]
          }}}))
     )
     V 2:Vector3f=&MakeVector(
      X=0.68
      Y=0.85
      Z=&FN_PickFloatDependOnPlatform(
       TargetPlatform=DataInterface() {propertyValues {Platforms {qualityLevelMask=-1, deviceProfileStates [{profileName=Mobile, qualityLevelMask=0, setQualityLevelMask=4294967295}]}}}
       TargetPlatformActive=0.7
       TargetPlatformNoActive=0.4
      )
     )
     V 3:Vector3f=?
     V 4:Vector3f=?
     V 5:Vector3f=?
     Vector Count:ENiagara_MultipleLerpCount=::NewEnumerator3
    )
   ) {isSetParameters=true, module=SetVariables_#5}
   /SetVariables(
    Particles.Phone Screen Glow Position:NiagaraPosition=&AddVectorToPosition(Position=$Engine.Emitter.SimulationPosition, Vector=<-12 2.75 -14.5>)
    Particles.Phone Screen Linear Color:LinearColor=&LerpLinearColors(
     StartColor=&MakeLinearColorFromVectorAndFloat(
      "Vector (RGB)"=&MakeVectorFromLinearColorRGB(Color=&MakeLinearColorFromVectorAndFloat(
        "Vector (RGB)"=&MakeVectorFromLinearColorRGB(Color=<rgba 0.2953 0.3684 1.5 1>)
        "Float (Alpha)"=&FloatFromCurve(
         FloatCurve=DataInterface() {propertyValues {Curve {
            defaultValue=1
            keys [{interpMode=RCIM_Cubic, tangentMode=RCTM_Break, time=0, value=0}, {interpMode=RCIM_Cubic, tangentMode=RCTM_Break, time=0.05, value=1}]
           }}}
         Scale Curve=&Waveform(
          Global Input Position=1
          Global Amplitude Scale=&Multiply_Float(
           A=&FNTimeOfDayScalar(Night Scale=0.5, Morning Scale=0.4, Day Scale=0.5, Evening Scale=0.4)
           B=$User.Light_Intensity
          )
          "Waveform [1]"=::NewEnumerator0
          "[1] Frequency"=0.5
          "[1] Amplitude Min/Max"=<0.8 1>
         )
        )
       ))
      "Float (Alpha)"=1
     )
     EndColor=&Multiply_LinearColorByFloat(
      Linear Color=<rgba 0.2953 0.3684 1.5 1>
      Float=&Waveform(
       Global Input Position=1
       Global Amplitude Scale=&Multiply_Float(
        A=&FNTimeOfDayScalar(Night Scale=0.56, Morning Scale=0.4, Day Scale=0.5, Evening Scale=0.4)
        B=$User.Light_Intensity
       )
       "Waveform [1]"=::NewEnumerator0
       "[1] Frequency"=0.5
       "[1] Amplitude Min/Max"=<0.8 1>
      )
     )
     LerpFactor=$User.on_true
    )
    Particles.Phone Screen SpriteAlignment:Vector3f=<0 0 1>
    Particles.Phone Screen SpriteFacing:Vector3f=<0 1 0.0131>
    Particles.Phone Screen SpriteSize:Vector2f=&Lerp_Vector2(
     A=&MakeVector2D(
      X=&Multiply_Float(
       A=&FloatFromCurve(FloatCurve=DataInterface() {propertyValues {Curve {defaultValue=1, keys [{tangentMode=RCTM_User, time=0, value=0}, {tangentMode=RCTM_User, time=0.02, value=1}]}}})
       B=18
      )
      Y=&Multiply_Float(
       A=&FloatFromCurve(FloatCurve=DataInterface() {propertyValues {Curve {defaultValue=1, keys [{tangentMode=RCTM_User, time=0, value=0}, {tangentMode=RCTM_User, time=0.02, value=1}]}}})
       B=30.5
      )
     )
     B=<13 25>
     Alpha=$User.on_true
    )
   ) {isSetParameters=true, module=SetVariables_#6}
   &FN_Light_Attributes(
    Color=<rgba 1.2969 3.0236 5 1>
    Create Color Attribute=true
    Create Exponent Attribute=true
    Create Position Attribute=true
    Create Radius Attribute=true
    Enable Particle Light Enable=true
    Enable Visibility Tag=true
    Exponent=5
    Light Enabled=true
    Position=<-15 22 15>
    Radius=25
    Scale Color=&Multiply_Float(A=0.85, B=$User.Light_Intensity)
   )
  ]
  rendererClasses [
   &NiagaraMeshRendererProperties
   &NiagaraLightRendererProperties
   &NiagaraSpriteRendererProperties
  ]
  renderers [
   {rendererClass=&NiagaraMeshRendererProperties, rendererIndex=0}
   {rendererClass=&NiagaraMeshRendererProperties, rendererIndex=1}
   {rendererClass=&NiagaraMeshRendererProperties, rendererIndex=2}
   {rendererClass=&NiagaraLightRendererProperties, rendererIndex=3}
   {rendererClass=&NiagaraSpriteRendererProperties, rendererIndex=4}
  ]
 }
 Minimal00__ISSUES_WITH_MOBILE {
  bEnabled=false
  emitterUpdateScript [
   &EmitterState(Inactive Response=::NewEnumerator1, Life Cycle Mode=::NewEnumerator1, Loop Duration=10, UseLoopDelay=false)
   &SpawnBurst_Instantaneous(
    Spawn Count=&Multiply_Int(A=1, B=&MakeIntFromFloat(Float=$User.ForceSpawn))
    Use Spawn Probability=false
    UseLoopCountLimit=false
   )
  ]
  particleSpawnScript [
   /SetVariables(Particles.VisibilityTag=&SpawnGroup()) {isSetParameters=true, module=SetVariables_#3}
   &InitializeParticle(Color Mode=::NewEnumerator0, Lifetime=10, Position Mode=::NewEnumerator1, UsePositionOffset=false, Write Material Random=true)
   &InitialMeshOrientation(
    Mesh Axis To Orient=<0 0 -1>
    Mesh Orientation Mode=::NewEnumerator2
    Orientation Coordinate Space=::Local
    Orientation Vector=&MergeHelper_VectorSwitch(0=<-0.1777 15 0.1184>, 1=<0 0 0>, 2=?, 3=?, Index=&SpawnGroup(), Num=::NewEnumerator0)
    Use Rotation Vector=false
   )
   &Color()
  ]
  particleUpdateScript [
   &ParticleState()
   /SetVariables(
    Particles.Core Green Color:LinearColor=&LerpLinearColors(
     StartColor=&MakeLinearColorFromVectorAndFloat(
      "Vector (RGB)"=&MakeVectorFromLinearColorRGB(Color=&MakeLinearColorFromVectorAndFloat(
        "Vector (RGB)"=&MakeVectorFromLinearColorRGB(Color=<rgba 0.3433 1 0 1>)
        "Float (Alpha)"=&FloatFromCurve(
         FloatCurve=DataInterface() {propertyValues {Curve {
            defaultValue=1
            keys [{interpMode=RCIM_Cubic, tangentMode=RCTM_Break, time=0, value=0}, {interpMode=RCIM_Cubic, tangentMode=RCTM_Break, time=0.05, value=1}]
           }}}
         Scale Curve=&Waveform(
          Global Input Position=1
          Global Amplitude Scale=&Multiply_Float(
           A=&FNTimeOfDayScalar(Night Scale=0.56, Morning Scale=0.45, Day Scale=0.56, Evening Scale=0.45)
           B=$User.Light_Intensity
          )
          "Waveform [1]"=::NewEnumerator0
          "[1] Frequency"=5
          "[1] Amplitude Min/Max"=<0.8 1>
          "[1] Scale"=0.5
         )
        )
       ))
      "Float (Alpha)"=1
     )
     EndColor=&Multiply_LinearColorByFloat(
      Linear Color=<rgba 1.0299 3 0 1>
      Float=&Waveform(
       Global Input Position=1
       Global Amplitude Scale=&Multiply_Float(
        A=&FNTimeOfDayScalar(Night Scale=0.7, Morning Scale=0.55, Day Scale=0.7, Evening Scale=0.55)
        B=$User.Light_Intensity
       )
       "Waveform [1]"=::NewEnumerator0
       "[1] Frequency"=5
       "[1] Amplitude Min/Max"=<0.8 1>
       "[1] Scale"=0.5
      )
     )
     LerpFactor=$User.on_true
    )
    Particles.Core Light Color:LinearColor=&LerpLinearColors(
     StartColor=&MakeLinearColorFromVectorAndFloat(
      "Vector (RGB)"=&MakeVectorFromLinearColorRGB(Color=&MakeLinearColorFromVectorAndFloat(
        "Vector (RGB)"=&MakeVectorFromLinearColorRGB(Color=<rgba 0.139 0.4339 5 1>)
        "Float (Alpha)"=&FloatFromCurve(
         FloatCurve=DataInterface() {propertyValues {Curve {
            defaultValue=1
            keys [{interpMode=RCIM_Cubic, tangentMode=RCTM_Break, time=0, value=0}, {interpMode=RCIM_Cubic, tangentMode=RCTM_Break, time=0.05, value=1}]
           }}}
         Scale Curve=&Waveform(
          Global Input Position=1
          Global Amplitude Scale=&Multiply_Float(
           A=&FNTimeOfDayScalar(Night Scale=0.7, Morning Scale=0.55, Day Scale=0.7, Evening Scale=0.55)
           B=$User.Light_Intensity
          )
          "Waveform [1]"=::NewEnumerator0
          "[1] Frequency"=5
          "[1] Amplitude Min/Max"=<0.8 1>
          "[1] Scale"=0.5
         )
        )
       ))
      "Float (Alpha)"=1
     )
     EndColor=&Multiply_LinearColorByFloat(
      Linear Color=<rgba 0.1668 0.5207 6 1>
      Float=&Waveform(
       Global Input Position=1
       Global Amplitude Scale=&Multiply_Float(
        A=&FNTimeOfDayScalar(Night Scale=0.7, Morning Scale=0.55, Day Scale=0.7, Evening Scale=0.55)
        B=$User.Light_Intensity
       )
       "Waveform [1]"=::NewEnumerator0
       "[1] Frequency"=5
       "[1] Amplitude Min/Max"=<0.8 1>
       "[1] Scale"=0.5
      )
     )
     LerpFactor=$User.on_true
    )
    Particles.Core Light Position:NiagaraPosition=&AddVectorToPosition(Position=$Engine.Emitter.SimulationPosition, Vector=<-12 -1.2 -14>)
    Particles.Core Light Scale:Vector3f=&LerpMultipleVectors(
     Alpha:NiagaraFloat=$User.on_true
     V 1:Vector3f=&Multiply_Vector(
      A=<0.73 0.85 0.7>
      B=&VectorFromFloat(Value=&FloatFromCurve(FloatCurve=DataInterface() {propertyValues {Curve {
           defaultValue=1
           keys [{interpMode=RCIM_Cubic, tangentMode=RCTM_Break, time=0, value=0}, {interpMode=RCIM_Cubic, tangentMode=RCTM_Break, time=0.02, value=1}]
          }}}))
     )
     V 2:Vector3f=<0.73 0.85 0.7>
     V 3:Vector3f=?
     V 4:Vector3f=?
     V 5:Vector3f=?
     Vector Count:ENiagara_MultipleLerpCount=::NewEnumerator3
    )
   ) {isSetParameters=true, module=SetVariables_#4}
   /SetVariables(
    Particles.Outer Glow Color:LinearColor=&LerpLinearColors(
     StartColor=&MakeLinearColorFromVectorAndFloat(
      "Vector (RGB)"=&MakeVectorFromLinearColorRGB(Color=&FNTimeOfDayColor(
        Night Color=<rgba 0.0688 0.1859 2 1>
        Morning Color=<rgba 0.0278 0.0868 1 1>
        Day Color=<rgba 0.0278 0.0868 1 1>
        Evening Color=<rgba 0.0278 0.0868 1 1>
       ))
      "Float (Alpha)"=&FloatFromCurve(
       FloatCurve=DataInterface() {propertyValues {Curve {
          defaultValue=1
          keys [{interpMode=RCIM_Cubic, tangentMode=RCTM_Break, time=0, value=1}, {interpMode=RCIM_Cubic, tangentMode=RCTM_Break, time=0.05, value=1}]
         }}}
       Scale Curve=&Waveform(
        Global Amplitude Scale=&FNTimeOfDayScalar(Night Scale=0.53, Morning Scale=0.35, Day Scale=0.67, Evening Scale=0.35)
        "Waveform [1]"=::NewEnumerator0
        "[1] Frequency"=1
        "[1] Amplitude Min/Max"=<0.85 1>
       )
      )
     )
     EndColor=&MakeLinearColorFromVectorAndFloat(
      "Vector (RGB)"=&MakeVectorFromLinearColorRGB(Color=&FNTimeOfDayColor(
        Night Color=<rgba 0.0688 0.1859 2 1>
        Morning Color=<rgba 0.0278 0.0868 1 1>
        Day Color=<rgba 0.0278 0.0868 1 1>
        Evening Color=<rgba 0.0278 0.0868 1 1>
       ))
      "Float (Alpha)"=&Waveform(
       Global Amplitude Scale=&FNTimeOfDayScalar(Night Scale=0.53, Morning Scale=0.35, Day Scale=0.67, Evening Scale=0.35)
       "Waveform [1]"=::NewEnumerator0
       "[1] Frequency"=1
       "[1] Amplitude Min/Max"=<0.85 1>
      )
     )
     LerpFactor=$User.on_true
    )
    Particles.Outer Glow Position:NiagaraPosition=&AddVectorToPosition(Position=$Engine.Emitter.SimulationPosition, Vector=<-12 -1.3 -14>)
    Particles.Outer Glow Scale:Vector3f=&LerpMultipleVectors(
     Alpha:NiagaraFloat=$User.on_true
     V 1:Vector3f=&Multiply_Vector(
      A=<0.68 0.85 0.7>
      B=&VectorFromFloat(Value=&FloatFromCurve(FloatCurve=DataInterface() {propertyValues {Curve {
           defaultValue=1
           keys [{interpMode=RCIM_Cubic, tangentMode=RCTM_Break, time=0, value=0}, {interpMode=RCIM_Cubic, tangentMode=RCTM_Break, time=0.05, value=1}]
          }}}))
     )
     V 2:Vector3f=<0.68 0.85 0.7>
     V 3:Vector3f=?
     V 4:Vector3f=?
     V 5:Vector3f=?
     Vector Count:ENiagara_MultipleLerpCount=::NewEnumerator3
    )
   ) {isSetParameters=true, module=SetVariables_#5}
   /SetVariables(
    Particles.Phone Screen Glow Position:NiagaraPosition=&AddVectorToPosition(Position=$Engine.Emitter.SimulationPosition, Vector=<-12 2.75 -14.5>)
    Particles.Phone Screen Linear Color:LinearColor=&LerpLinearColors(
     StartColor=&MakeLinearColorFromVectorAndFloat(
      "Vector (RGB)"=&MakeVectorFromLinearColorRGB(Color=&MakeLinearColorFromVectorAndFloat(
        "Vector (RGB)"=&MakeVectorFromLinearColorRGB(Color=<rgba 0.2953 0.3684 1.5 1>)
        "Float (Alpha)"=&FloatFromCurve(
         FloatCurve=DataInterface() {propertyValues {Curve {
            defaultValue=1
            keys [{interpMode=RCIM_Cubic, tangentMode=RCTM_Break, time=0, value=0}, {interpMode=RCIM_Cubic, tangentMode=RCTM_Break, time=0.05, value=1}]
           }}}
         Scale Curve=&Waveform(
          Global Input Position=1
          Global Amplitude Scale=&Multiply_Float(
           A=&FNTimeOfDayScalar(Night Scale=0.5, Morning Scale=0.4, Day Scale=0.5, Evening Scale=0.4)
           B=$User.Light_Intensity
          )
          "Waveform [1]"=::NewEnumerator0
          "[1] Frequency"=0.5
          "[1] Amplitude Min/Max"=<0.8 1>
         )
        )
       ))
      "Float (Alpha)"=1
     )
     EndColor=&Multiply_LinearColorByFloat(
      Linear Color=<rgba 0.2953 0.3684 1.5 1>
      Float=&Waveform(
       Global Input Position=1
       Global Amplitude Scale=&Multiply_Float(
        A=&FNTimeOfDayScalar(Night Scale=0.56, Morning Scale=0.4, Day Scale=0.5, Evening Scale=0.4)
        B=$User.Light_Intensity
       )
       "Waveform [1]"=::NewEnumerator0
       "[1] Frequency"=0.5
       "[1] Amplitude Min/Max"=<0.8 1>
      )
     )
     LerpFactor=$User.on_true
    )
    Particles.Phone Screen SpriteAlignment:Vector3f=<0 0 1>
    Particles.Phone Screen SpriteFacing:Vector3f=<0 1 0.0131>
    Particles.Phone Screen SpriteSize:Vector2f=&Lerp_Vector2(
     A=&MakeVector2D(
      X=&Multiply_Float(
       A=&FloatFromCurve(FloatCurve=DataInterface() {propertyValues {Curve {defaultValue=1, keys [{tangentMode=RCTM_User, time=0, value=0}, {tangentMode=RCTM_User, time=0.02, value=1}]}}})
       B=18
      )
      Y=&Multiply_Float(
       A=&FloatFromCurve(FloatCurve=DataInterface() {propertyValues {Curve {defaultValue=1, keys [{tangentMode=RCTM_User, time=0, value=0}, {tangentMode=RCTM_User, time=0.02, value=1}]}}})
       B=30.5
      )
     )
     B=<13 25>
     Alpha=$User.on_true
    )
   ) {isSetParameters=true, module=SetVariables_#6}
   &FN_Light_Attributes(
    Color=<rgba 1.2969 3.0236 5 1>
    Create Color Attribute=true
    Create Exponent Attribute=true
    Create Position Attribute=true
    Create Radius Attribute=true
    Enable Particle Light Enable=true
    Enable Visibility Tag=true
    Exponent=5
    Light Enabled=true
    Position=<-15 22 15>
    Radius=25
    Scale Color=&Multiply_Float(A=0.85, B=$User.Light_Intensity)
   )
  ]
  rendererClasses [
   &NiagaraMeshRendererProperties
   &NiagaraLightRendererProperties
   &NiagaraSpriteRendererProperties
  ]
  renderers [
   {rendererClass=&NiagaraMeshRendererProperties, rendererIndex=0}
   {rendererClass=&NiagaraMeshRendererProperties, rendererIndex=1}
   {rendererClass=&NiagaraMeshRendererProperties, rendererIndex=2}
   {rendererClass=&NiagaraLightRendererProperties, rendererIndex=3}
   {rendererClass=&NiagaraSpriteRendererProperties, rendererIndex=4}
  ]
 }
 motion_blur_hand {
  emitterUpdateScript [
   &EmitterState(
    Inactive Response=::NewEnumerator1
    Life Cycle Mode=::NewEnumerator1
    Loop Behavior=::NewEnumerator1
    Loop Duration Mode=::NewEnumerator1
    UseLoopDelay=false
   )
   &FN_SpawnAnimTrail(Spawn Mode:ENiagaraSkeletalSpawning=::NewEnumerator0, Switch on ENiagara_AnimTrail_Source:ENiagara_AnimTrail_Source=::NewEnumerator1)
  ]
  particleSpawnScript [
   &FN_InitializeAnimTrailBoneIndecies()
   &FN_InitializeAnimTrail(
    AnimTrailWidth=11
    Lifetime=&Multiply_Float(
     A=&FloatFromCurve(
      FloatCurve=DataInterface() {propertyValues {Curve {
         defaultValue=1
         keys [{interpMode=RCIM_Cubic, tangentMode=RCTM_User, time=0, value=1}, {interpMode=RCIM_Cubic, tangentMode=RCTM_User, time=0.4905, value=-0.0601}]
        }}}
      Scale Curve=0.1
     )
     B=$User.AnimTrail_SpawnRate
    )
   )
   &SkeletalMeshLocation(Bone Sampling Mode=::NewEnumerator4, MaskSpawnGroup=false, Skeletal Mesh=$System.SkeletalMesh)
  ]
  particleUpdateScript [
   &ParticleState()
   &ScaleColor(
    Scale Alpha=&FloatFromCurve(
     FloatCurve=DataInterface() {propertyValues {Curve {keys [{arriveTangent=-3.0806, interpMode=RCIM_Cubic, leaveTangent=-3.0806, tangentMode=RCTM_User, time=0, value=1}, <at 0.85 0.2071 RCIM_Cubic>]}}}
     Scale Curve=0.5
    )
    Scale RGB=&MakeVectorFromLinearColorRGB(Color=&ColorFromCurve(ColorCurve=DataInterface() {propertyValues {
        AlphaCurve {keys [<at 0 0.8 RCIM_Cubic>,<at 0.4686 0 RCIM_Cubic>]}
        BlueCurve {keys [<at 0 0.2 RCIM_Cubic>]}
        GreenCurve {keys [<at 0 0.1694 RCIM_Cubic>]}
        RedCurve {keys [<at 0 0.1319 RCIM_Cubic>]}
       }}))
    ScaleA=true
    ScaleRGB=true
   ) {module=ScaleColor003}
   &ScaleRibbonWidth(Ribbon Width Scale=&FloatFromCurve(
     FloatCurve=DataInterface() {propertyValues {Curve {
        defaultValue=1
        keys [
         {arriveTangent=0.0733, interpMode=RCIM_Cubic, leaveTangent=0.0733, tangentMode=RCTM_User, time=0, value=1}
         {arriveTangent=-0.0405, interpMode=RCIM_Cubic, leaveTangent=-0.0405, tangentMode=RCTM_User, time=0.7, value=0.3}
        ]
       }}}
     Scale Curve=1.7
    ))
   &InheritSourceMovement(ApplyForce=false, ApplyPosition=true, ApplyPositionFalloff=false)
   &Drag(Use Linear Drag=true, Use Rotational Drag=false)
   &SolveForcesAndVelocity(Clamp Velocity=false, Limit Acceleration=false)
   /SetVariables(Particles.Motion Blur color:LinearColor=/DealershipSharedLibrary/FX/Niagara/DynamicInputs/MultiplyLinearColorByVectorAndFloat(
     Color to Scale:LinearColor=&ColorFromCurve(ColorCurve=DataInterface() {propertyValues {
        AlphaCurve {keys [<at 0 0.5 RCIM_Cubic>,<at 0.9052 0 RCIM_Cubic>]}
        BlueCurve {keys [<at 0 1 RCIM_Cubic>]}
        GreenCurve {keys [<at 0 0.7939 RCIM_Cubic>]}
        RedCurve {keys [<at 0 0.6135 RCIM_Cubic>]}
       }})
     RGB Scale:Vector3f=<1 1 1>
     Alpha Scale:NiagaraFloat=&FloatFromCurve(
      FloatCurve=DataInterface() {propertyValues {Curve {keys [{arriveTangent=-3.0806, interpMode=RCIM_Cubic, leaveTangent=-3.0806, tangentMode=RCTM_User, time=0, value=1}, <at 0.85 0.2071 RCIM_Cubic>]}}}
      Scale Curve=&FNTimeOfDayScalar(Night Scale=1, Morning Scale=0.75, Day Scale=0.75, Evening Scale=0.75)
     )
    )) {isSetParameters=true}
   &FN_UpdateAnimTrail()
  ]
  rendererClasses [&NiagaraRibbonRendererProperties]
  renderers [{rendererClass=&NiagaraRibbonRendererProperties}]
 }
}
summary {
 systemName=NS_Emote_MicroLeaf_Light
 emitters [
  {emitterName=motion_blur_hand, rendererClasses [&NiagaraRibbonRendererProperties]}
  {
   bEnabled=false
   emitterName=Minimal00__ISSUES_WITH_MOBILE
   rendererClasses [
    &NiagaraMeshRendererProperties
    &NiagaraLightRendererProperties
    &NiagaraSpriteRendererProperties
   ]
  }
  {
   emitterName=Minimal
   rendererClasses [
    &NiagaraMeshRendererProperties
    &NiagaraLightRendererProperties
    &NiagaraSpriteRendererProperties
   ]
  }
 ]
 userVariables [
  User.AnimTrail_SpawnRate {type=NiagaraFloat, defaultValue=0, description=""}
  User.ForceSpawn {type=NiagaraFloat, defaultValue=1, description=""}
  User.Light_Intensity {type=NiagaraFloat, defaultValue=0, description=""}
  User.NormalizedNotifyProgress {type=NiagaraFloat, defaultValue=0, description=""}
  User.on_true {type=NiagaraFloat, defaultValue=0, description=""}
  User.RibbonIDOffset {type=NiagaraInt32}
 ]
}
